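# Sound cue for the looping acceleration sound of the jetpack item.
# A quality level node selects between two looping wave players: the standard wave for the first two levels and an "opt" variant for the last two.
# The cue plays in the gadget items sound class with jetpack attenuation, limited by both an accel-loop concurrency group and a global jetpack group.
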SoundCue /FNGameplayCues/Athena/Sounds/Items/Jetpack/Jetpack_AccelLoop_Cue
root=QualityLevel_0
nodes [
 {id=QualityLevel_0, children [WavePlayer_0,WavePlayer_0,WavePlayer_1,WavePlayer_1]}
 {id=WavePlayer_0, properties {bLooping=true, soundWaveAssetPtr=&/Jetpack_AccelLoop_03}}
 {id=WavePlayer_1, properties {bLooping=true, soundWaveAssetPtr=&/Jetpack_AccelLoop_03_opt}}
]
properties {
 attenuationSettings=/Game/Sounds/Attenuation/Fort_Jetpack_Att
 maxDistance=13067
 soundClassObject=&Fort_Gadgets_Items
 volumeMultiplier=0.6
 concurrencySet [&/CONC_Jetpack_Accel_Loop, &/CONC_Jetpack_Global]
}
