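# Looping, attenuated sound cue for jetpack fuel charging, routed to the gadget items sound class.
# A parameter crossfade driven by "Fader" blends from a constant fuel-up loop to an energy counter loop borrowed from front-end UI sounds as the value goes from 0 to 1, then slowly fades the latter out above 1.
# A mixer boosts the fuel-up loop, while a modulator randomizes the counter loop's pitch slightly upward.
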SoundCue /FNGameplayCues/Athena/Sounds/Items/Jetpack/Jetpack_FuelChargeLoop_Cue
root=ParamCrossFade_0
nodes [
 {id=Mixer_1, children [Modulator_1], properties {inputVolume [3]}}
 {id=Modulator_0, children [WavePlayer_1], properties {pitchMax=1.2, pitchMin=1}}
 {id=Modulator_1, children [WavePlayer_0], properties {pitchMax=1, pitchMin=1}}
 {
  id=ParamCrossFade_0
  children [Mixer_1,Modulator_0]
  properties {
   paramName=Fader
   crossFadeInput [
    {fadeInDistanceEnd=0, fadeInDistanceStart=0, fadeOutDistanceEnd=1, fadeOutDistanceStart=0, volume=1}
    {fadeInDistanceEnd=1, fadeInDistanceStart=0, fadeOutDistanceEnd=100, fadeOutDistanceStart=1, volume=1}
   ]
  }
 }
 {id=WavePlayer_0, properties {bLooping=true, soundWaveAssetPtr=&/Jetpack_FuelUp_ConstantLoop_01}}
 {id=WavePlayer_1, properties {bLooping=true, soundWaveAssetPtr=/Game/Sounds/Fort_UI_Sounds/FrontEnd/Counter_Energy_Loop08}}
]
properties {
 attenuationSettings=/Game/Sounds/Attenuation/Fort_Jetpack_Fuel_Att
 maxDistance=800
 pitchMultiplier=1.1
 soundClassObject=&Fort_Gadgets_Items
 volumeMultiplier=0.6
}
