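# Sound cue for the Athena jetpack item's fuel-up UI beep.
# A Random node picks between two equally weighted beep wave players.
# It routes to the gadgets items sound class with a jetpack fuel attenuation asset, priming on load and playing slightly below full volume.
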SoundCue /FNGameplayCues/Athena/Sounds/Items/Jetpack/Jetpack_FuelUp_Cue
root=Random_0
nodes [
 {id=Random_0, children [WavePlayer_0,WavePlayer_2], properties {weights [1,1]}}
 {id=WavePlayer_0, properties {soundWaveAssetPtr=/Game/Athena/Sounds/Items/Jetpack/Jetpack_FuelUp_UIBeep_01}}
 {id=WavePlayer_2, properties {soundWaveAssetPtr=/Game/Athena/Sounds/Items/Jetpack/Jetpack_FuelUp_UIBeep_02}}
]
properties {
 attenuationSettings=/Game/Sounds/Attenuation/Fort_Jetpack_Fuel_Att
 bPrimeOnLoad=true
 maxDistance=800
 soundClassObject=&Fort_Gadgets_Items
 volumeMultiplier=0.9
}
