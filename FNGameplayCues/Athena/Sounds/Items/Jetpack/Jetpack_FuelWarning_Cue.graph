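# Sound cue for the jetpack fuel warning, classed under gadget item sounds.
# A single wave player is the root and plays one warning wave with no randomization or modulation, at slightly reduced volume.
# It uses the shared jetpack attenuation asset and overrides concurrency so only one instance plays at a time.
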SoundCue /FNGameplayCues/Athena/Sounds/Items/Jetpack/Jetpack_FuelWarning_Cue
root=WavePlayer_2
nodes [{id=WavePlayer_2, properties {soundWaveAssetPtr=&/Jetpack_FuelWarning_03}}]
properties {
 attenuationSettings=/Game/Sounds/Attenuation/Fort_Jetpack_Att
 bOverrideConcurrency=true
 maxDistance=13067
 soundClassObject=&Fort_Gadgets_Items
 volumeMultiplier=0.85
 concurrencyOverrides {maxCount=1}
}
