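# Looping idle sound cue for the Papaya jetpack item, played in the Fort_Gadgets_Items sound class at reduced volume.
# A quality-level node at the root selects between two looping wave players: the first level plays the standard idle loop, and the other three share the "_opt" wave.
# A dedicated Papaya jetpack attenuation asset handles spatial falloff.
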SoundCue /FNGameplayCues/Athena/Sounds/Items/Jetpack/Jetpack_IdleLoop_Papaya_Cue
root=QualityLevel_0
nodes [
 {id=QualityLevel_0, children [WavePlayer_0,WavePlayer_1,WavePlayer_1,WavePlayer_1]}
 {id=WavePlayer_0, properties {bLooping=true, soundWaveAssetPtr=&/Jetpack_IdleLoop_01}}
 {id=WavePlayer_1, properties {bLooping=true, soundWaveAssetPtr=&/Jetpack_IdleLoop_01_opt}}
]
properties {
 attenuationSettings=/FNGameplayCues/Sounds/Attenuation/Fort_Jetpack_Papaya_Att
 maxDistance=3000
 soundClassObject=&Fort_Gadgets_Items
 volumeMultiplier=0.4
}
