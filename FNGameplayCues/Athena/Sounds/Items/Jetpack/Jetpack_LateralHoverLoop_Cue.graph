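# Sound cue for the looping lateral hover sound of the jetpack item.
# A single looping wave player plays the lateral hover loop wave at reduced volume.
# It uses the jetpack attenuation asset and is routed to the gadget items sound class.
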SoundCue /FNGameplayCues/Athena/Sounds/Items/Jetpack/Jetpack_LateralHoverLoop_Cue
root=WavePlayer_2
nodes [{id=WavePlayer_2, properties {bLooping=true, soundWaveAssetPtr=&/Jetpack_LateralHoverLoop_01}}]
properties {
 attenuationSettings=/Game/Sounds/Attenuation/Fort_Jetpack_Att
 maxDistance=13067
 soundClassObject=&Fort_Gadgets_Items
 volumeMultiplier=0.5
}
