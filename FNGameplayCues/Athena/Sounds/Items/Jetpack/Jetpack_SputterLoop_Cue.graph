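# Sound cue for the looping sputter sound of the jetpack item, classed under gadget item sounds with jetpack attenuation.
# Each mixer layers two looping waves, a quieter base sputter loop and a louder layer loop.
# A quality-level node feeds the standard mix to the first two tiers and an "_opt" variant mix to the last two.
# A concurrency override caps simultaneous instances.
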SoundCue /FNGameplayCues/Athena/Sounds/Items/Jetpack/Jetpack_SputterLoop_Cue
root=QualityLevel_0
nodes [
 {id=Mixer_0, children [WavePlayer_1,WavePlayer_0], properties {inputVolume [0.2,0.8]}}
 {id=Mixer_1, children [WavePlayer_5,WavePlayer_4], properties {inputVolume [0.2,0.8]}}
 {id=QualityLevel_0, children [Mixer_0,Mixer_0,Mixer_1,Mixer_1]}
 {id=WavePlayer_0, properties {bLooping=true, soundWaveAssetPtr=&/Jetpack_SputterLoopLayer_01}}
 {id=WavePlayer_1, properties {bLooping=true, soundWaveAssetPtr=&/Jetpack_SputterLoop_01}}
 {id=WavePlayer_4, properties {bLooping=true, soundWaveAssetPtr=&/Jetpack_SputterLoopLayer_01_opt}}
 {id=WavePlayer_5, properties {bLooping=true, soundWaveAssetPtr=&/Jetpack_SputterLoop_01_opt}}
]
properties {
 attenuationSettings=/Game/Sounds/Attenuation/Fort_Jetpack_Att
 bOverrideConcurrency=true
 maxDistance=13067
 soundClassObject=&Fort_Gadgets_Items
 volumeMultiplier=0.7
 concurrencyOverrides {maxCount=3}
}
