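# Sound cue for the start of a Papaya-variant jetpack item, classed under gadget item sounds.
# A single wave player plays one jetpack start wave through a modulator that randomizes pitch slightly upward for variation.
# Playback is quieted by a reduced volume multiplier and spatialized with a dedicated Papaya jetpack attenuation asset.
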SoundCue /FNGameplayCues/Athena/Sounds/Items/Jetpack/Jetpack_Start_Papaya_Cue
root=Modulator_0
nodes [
 {id=Modulator_0, children [WavePlayer_1], properties {pitchMax=1.1, pitchMin=1}}
 {id=WavePlayer_1, properties {soundWaveAssetPtr=&/Jetpack_Start_01}}
]
properties {
 attenuationSettings=/FNGameplayCues/Sounds/Attenuation/Fort_Jetpack_Papaya_Att
 maxDistance=3000
 soundClassObject=&Fort_Gadgets_Items
 volumeMultiplier=0.4
}
