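# Sound cue for the Athena jetpack thrust sound.
# A Random node picks one of three equally weighted thrust wave variations, and a Modulator at the root adds slight upward pitch variation.
# The cue plays at reduced volume under the gadget items sound class, with jetpack attenuation and both thrust-specific and global jetpack concurrency limits.
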SoundCue /FNGameplayCues/Athena/Sounds/Items/Jetpack/Jetpack_Thrust_Cue
root=Modulator_0
nodes [
 {id=Modulator_0, children [Random_0], properties {pitchMax=1.1, pitchMin=1}}
 {id=Random_0, children [WavePlayer_1,WavePlayer_2,WavePlayer_3], properties {weights [1,1,1]}}
 {id=WavePlayer_1, properties {soundWaveAssetPtr=/Game/Athena/Sounds/Items/Jetpack/Jetpack_Trust_01}}
 {id=WavePlayer_2, properties {soundWaveAssetPtr=/Game/Athena/Sounds/Items/Jetpack/Jetpack_Trust_02}}
 {id=WavePlayer_3, properties {soundWaveAssetPtr=/Game/Athena/Sounds/Items/Jetpack/Jetpack_Trust_03}}
]
properties {
 attenuationSettings=/Game/Sounds/Attenuation/Fort_Jetpack_Att
 maxDistance=13067
 soundClassObject=&Fort_Gadgets_Items
 volumeMultiplier=0.5
 concurrencySet [&/CONC_Jetpack_Thrust, &/CONC_Jetpack_Global]
}
